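# Masked, unlit material in the damage text UI base folder, named as a transparent material.
# The graph has no textures or parameters, only a scalar constant at its default and a constant vector with an extremely large negative third component.
# The connections from these constants to material inputs are not shown, so how transparency is achieved is not visible.
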Material /Game/UI/DamageText/Base/TransparentMaterial
graph {
 properties {blendMode=BLEND_Masked, pixelDepthOffsetMode=PDOM_Legacy, shadingModel=MSM_Unlit, usedShadingModels=Unlit}
 nodes [
  Constant3Vector() {constant=<rgba 0 0 -10000000000 0>}
  Constant()
 ]
}
