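# Material function that drives World Position Offset for animated floating damage numbers, drawn as up to nine camera-facing sprite quads.
# Floored UV1 picks each quad's placement, scale, digit, timing and random values from vector parameters through chained lerps.
# Timing over Animation Lifespan samples two curve textures for movement, scale, rotation and brightness.
# Sprites get pushed toward the camera, and extra outputs supply custom UVs, normals and critical-hit data.
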MaterialFunction /Game/UI/DamageText/Base/DamageNumberWPO
graph {nodes [
  n1 = Multiply(A=Time()) {constB=0.25}
  Ceil(None=ConstantBiasScale(None=Frac(None=@n1)) {bias=-0.5, scale=1.5})
  n2 = VectorParameter() {desc=alpha contains offsets, group=Placement and offsets, parameterName=0a, defaultValue=<rgba 0 0 0 0.5>}
  n3 = VectorParameter() {group=Placement and offsets, parameterName=1a, defaultValue=<rgba 0 0 0 1.2521>}
  n4 = TextureCoordinate() {coordinateIndex=1}
  n5 = Floor(None=ComponentMask(None=FunctionInput(Preview=@n4) {bUsePreviewValueAsDefault=true, inputName=Debug texture coordinate input, inputType=FunctionInput_Vector2}) {r=true})
  n6 = VectorParameter() {group=Placement and offsets, parameterName=2a, defaultValue=<rgba 0 0 0 2.0847>}
  n7 = Subtract(A=@n5)
  n8 = Clamp(None=@n7)
  n9 = LinearInterpolate(
   A=LinearInterpolate(A=AppendVector(A=@n2.RGB, B=@n2.RGB), B=AppendVector(A=@n3.RGB, B=@n3.RGB), Alpha=@n5)
   B=AppendVector(A=@n6.RGB, B=@n6.RGB)
   Alpha=@n8
  )
  n10 = VectorParameter() {group=Placement and offsets, parameterName=3a, defaultValue=<rgba 0 0 0 2.9269>}
  n11 = AppendVector(A=@n10.RGB, B=@n10.RGB)
  n12 = Subtract(A=@n7)
  n13 = Clamp(None=@n12)
  n14 = VectorParameter() {group=Placement and offsets, parameterName=4a, defaultValue=<rgba 0 0 0 3.6974>}
  n15 = Subtract(A=@n12)
  n16 = Clamp(None=@n15)
  n17 = VectorParameter() {group=Placement and offsets, parameterName=5a, defaultValue=<rgba 0 0 0 1.2521>}
  n18 = Subtract(A=@n15)
  n19 = Clamp(None=@n18)
  n20 = LinearInterpolate(
   A=LinearInterpolate(A=LinearInterpolate(A=@n9, B=@n11, Alpha=@n13), B=AppendVector(A=@n14.RGB, B=@n14.RGB), Alpha=@n16)
   B=AppendVector(A=@n17.RGB, B=@n17.RGB)
   Alpha=@n19
  )
  n21 = VectorParameter() {group=Placement and offsets, parameterName=6a, defaultValue=<rgba 0 0 0 2.0847>}
  n22 = AppendVector(A=@n21.RGB, B=@n21.RGB)
  n23 = Subtract(A=@n18)
  n24 = Clamp(None=@n23)
  n25 = VectorParameter() {group=Placement and offsets, parameterName=7a, defaultValue=<rgba 0 0 0 2.9269>}
  n26 = Subtract(A=@n23)
  n27 = Clamp(None=@n26)
  n28 = VectorParameter() {group=Placement and offsets, parameterName=8a, defaultValue=<rgba 0 0 0 3.6974>}
  n29 = Clamp(None=Subtract(A=@n26))
  n30 = LinearInterpolate(
   A=LinearInterpolate(A=LinearInterpolate(A=@n20, B=@n22, Alpha=@n24), B=AppendVector(A=@n25.RGB, B=@n25.RGB), Alpha=@n27)
   B=AppendVector(A=@n28.RGB, B=@n28.RGB)
   Alpha=@n29
  )
  n31 = ComponentMask(None=@n30) {b=true, g=true, r=true}
  FunctionOutput(None=@n31) {outputName=Optional position info, sortPriority=4}
  n32 = MaterialFunctionCall(Float2=TextureCoordinate() {vTiling=0.1}) {materialFunction=&BreakOutFloat2Components}
  n33 = VectorParameter() {group="Scale, rotation angle and values", parameterName=0b, defaultValue=<rgba 30 50 0 0>}
  n34 = VectorParameter() {group="Scale, rotation angle and values", parameterName=1b, defaultValue=<rgba 30 50 0 1>}
  n35 = VectorParameter() {group="Scale, rotation angle and values", parameterName=2b, defaultValue=<rgba 30 50 0 1>}
  n36 = LinearInterpolate(
   A=LinearInterpolate(A=AppendVector(A=@n33.RGB, B=@n33.RGB), B=AppendVector(A=@n34.RGB, B=@n34.RGB), Alpha=@n5)
   B=AppendVector(A=@n35.RGB, B=@n35.RGB)
   Alpha=@n8
  )
  n37 = VectorParameter() {group="Scale, rotation angle and values", parameterName=3b, defaultValue=<rgba 30 50 2 -1>}
  n38 = AppendVector(A=@n37.RGB, B=@n37.RGB)
  n39 = VectorParameter() {group="Scale, rotation angle and values", parameterName=4b, defaultValue=<rgba 30 50 1 1>}
  n40 = VectorParameter() {group="Scale, rotation angle and values", parameterName=5b, defaultValue=<rgba 30 50 0 0>}
  n41 = LinearInterpolate(
   A=LinearInterpolate(A=LinearInterpolate(A=@n36, B=@n38, Alpha=@n13), B=AppendVector(A=@n39.RGB, B=@n39.RGB), Alpha=@n16)
   B=AppendVector(A=@n40.RGB, B=@n40.RGB)
   Alpha=@n19
  )
  n42 = VectorParameter() {group="Scale, rotation angle and values", parameterName=6b, defaultValue=<rgba 30 50 0 1>}
  n43 = AppendVector(A=@n42.RGB, B=@n42.RGB)
  n44 = VectorParameter() {group="Scale, rotation angle and values", parameterName=7b, defaultValue=<rgba 30 50 0 1>}
  n45 = VectorParameter() {group="Scale, rotation angle and values", parameterName=8b, defaultValue=<rgba 30 50 2 -1>}
  n46 = LinearInterpolate(
   A=LinearInterpolate(A=LinearInterpolate(A=@n41, B=@n43, Alpha=@n24), B=AppendVector(A=@n44.RGB, B=@n44.RGB), Alpha=@n27)
   B=AppendVector(A=@n45.RGB, B=@n45.RGB)
   Alpha=@n29
  )
  n47 = MaterialFunctionCall(Float4=@n46) {desc="rg = scale xy, b = rotation angle ", materialFunction=&BreakOutFloat4Components}
  n48 = Multiply(A=@n47.B) {constB=0.1}
  FunctionOutput(None=AppendVector(A=@n32.R, B=Add(A=@n32.G, B=@n48))) {outputName=CustomUV1, sortPriority=10}
  FunctionOutput(None=TextureCoordinate() {coordinateIndex=4}) {outputName=Up vector - UV4, sortPriority=16}
  n49 = CameraPositionWS()
  n50 = VectorParameter() {group=+-, parameterName=0c, defaultValue=<rgba -100 0.5575 0 0>}
  n51 = AppendVector(A=@n50.RGB, B=@n50.RGB)
  n52 = VectorParameter() {desc=0, group=+-, parameterName=1c, defaultValue=<rgba -100 0.8965 0 0>}
  n53 = AppendVector(A=@n52.RGB, B=@n52.RGB)
  n54 = VectorParameter() {desc=0, group=+-, parameterName=2c, defaultValue=<rgba -100 0.2989 0 0>}
  n55 = VectorParameter() {desc=0, group=+-, parameterName=3c, defaultValue=<rgba -100 0.179 0 0>}
  n56 = LinearInterpolate(
   A=LinearInterpolate(A=LinearInterpolate(A=@n51, B=@n53, Alpha=@n5), B=AppendVector(A=@n54.RGB, B=@n54.RGB), Alpha=@n8)
   B=AppendVector(A=@n55.RGB, B=@n55.RGB)
   Alpha=@n13
  )
  n57 = VectorParameter() {desc=0, group=+-, parameterName=4c, defaultValue=<rgba -100 0.3333 0 0>}
  n58 = AppendVector(A=@n57.RGB, B=@n57.RGB)
  n59 = VectorParameter() {group="Scale, rotation angle and values", parameterName=5c, defaultValue=<rgba -100 50 0 0>}
  n60 = VectorParameter() {group="Scale, rotation angle and values", parameterName=6c, defaultValue=<rgba -100 50 0 1>}
  n61 = LinearInterpolate(
   A=LinearInterpolate(A=LinearInterpolate(A=@n56, B=@n58, Alpha=@n16), B=AppendVector(A=@n59.RGB, B=@n59.RGB), Alpha=@n19)
   B=AppendVector(A=@n60.RGB, B=@n60.RGB)
   Alpha=@n24
  )
  n62 = VectorParameter() {group="Scale, rotation angle and values", parameterName=7c, defaultValue=<rgba -100 50 0 1>}
  n63 = AppendVector(A=@n62.RGB, B=@n62.RGB)
  n64 = VectorParameter() {group="Scale, rotation angle and values", parameterName=8c, defaultValue=<rgba -100 50 2 -1>}
  n65 = MaterialFunctionCall(Float4=LinearInterpolate(A=LinearInterpolate(A=@n61, B=@n63, Alpha=@n27), B=AppendVector(A=@n64.RGB, B=@n64.RGB), Alpha=@n29)) {materialFunction=&BreakOutFloat4Components}
  n66 = Divide(A=Subtract(A=@n65.R, B=Time() {bIgnorePause=true}), B=ScalarParameter() {defaultValue=1, parameterName=Animation Lifespan})
  n67 = FunctionInput(Preview=OneMinus(None=Clamp(None=@n66))) {bUsePreviewValueAsDefault=true, inputName=Animation, inputType=FunctionInput_Scalar, sortPriority=2}
  n68 = TextureSample(UVs=@n67) {desc=RGB Damage Curve, samplerType=SAMPLERTYPE_LinearColor, texture=&/AnimatedNumbersDamageCurve}
  n69 = Constant()
  n70 = Add(A=@n65.G) {desc=+ random value per quad}
  n71 = LinearInterpolate(A=@n68.B, B=Multiply(A=@n68.B, B=Multiply(A=@n70, B=@n68.G)) {desc=randomize the up swing height of each particle}, Alpha=@n68.B)
  n72 = Multiply(
   A=ScalarParameter() {defaultValue=50, parameterName=ambient movment distance}
   B=Add(
    A=Multiply(A=@n31, B=@n68.R) {desc=* random movement vectors}
    B=MaterialFunctionCall(X=@n69, Y=@n69, Z=@n71) {materialFunction=&MakeFloat3}.Result
   ) {desc=animation}
  ) {desc=" number locations"}
  n73 = Constant3Vector() {constant=<rgba 1 0 0 0>}
  n74 = ComponentMask(None=@n30) {a=true, desc=offset}
  n75 = ComponentMask(None=@n46) {g=true, r=true}
  n76 = TextureSample(UVs=@n67) {
   desc="XY Scale, Y Scale,  Rotation"
   samplerType=SAMPLERTYPE_LinearColor
   texture=&/ImpactNumbers_XYScale_Rotation_Texture
  }
  n77 = Constant() {r=1}
  n78 = MaterialFunctionCall(Float2=Multiply(A=Multiply(A=@n75, B=@n76.R) {desc=" XY dimensions"}, B=AppendVector(A=@n77, B=@n76.G))) {materialFunction=&BreakOutFloat2Components}
  n79 = MaterialFunctionCall(Float4=MaterialFunctionCall(X=@n78.R, Y=@n78.R, Z=Multiply(A=@n76.B, B=@n47.A)) {materialFunction=&MakeFloat4}.Result) {desc="rg = scale xy, b = rotation angle ", materialFunction=&BreakOutFloat4Components}
  n80 = AppendVector(A=@n79.R, B=@n79.R)
  n81 = FunctionInput() {
   bUsePreviewValueAsDefault=true
   inputName="Additonal Screenspace XY offset "
   inputType=FunctionInput_Vector2
   sortPriority=11
   previewValue=<0 0 0 0>
  }
  n82 = MaterialFunctionCall(Float2=Add(A=@n80, B=@n81)) {materialFunction=&BreakOutFloat2Components}
  n83 = MaterialFunctionCall(Y=@n82.G) {materialFunction=&MakeFloat3}
  n84 = Add(
   A=Multiply(A=Transform(None=@n73) {transformSourceType=TRANSFORMSOURCE_View}, B=Multiply(A=@n74, B=@n82.R) {desc=scale x by offset amount - accum 1s}) {constB=50}
   B=Transform(None=Multiply(A=@n83.Result) {constB=0.5}) {transformSourceType=TRANSFORMSOURCE_View}
  )
  n85 = FunctionInput() {bUsePreviewValueAsDefault=true, inputName=Additonal Scale, inputType=FunctionInput_Vector2, sortPriority=4, previewValue=<1 1 0 1>}
  n86 = MaterialFunctionCall!(Center Location=Add(A=@n72, B=@n84), 0-1 UVs=TextureCoordinate(), XY scale=Multiply(A=@n85, B=@n80), OverridePivot, "Normals (Optional)") {materialFunction=&Sprite}
  n87 = MaterialFunctionCall() {materialFunction=&CameraDirectionVector}
  n88 = Add(A=@n84, B=@n72)
  n89 = Add(A=@n86.World position offset, B=WorldPosition().XYZ)
  n90 = Add(
   A=MaterialFunctionCall() {materialFunction=&ComponentPivotLocation}.Result
   B=Add(A=@n86.World position offset, B=RotateAboutAxis(NormalizedRotationAxis=@n87.Result, RotationAngle=@n79.B, PivotPoint=@n88, Position=@n89))
  )
  n91 = Add(A=@n90, B=WorldPosition().XYZ)
  FunctionOutput(None=Normalize(VectorInput=Subtract(A=@n49, B=@n91))) {outputName=Debug Output}
  FunctionOutput(None=@n67) {outputName=Time, sortPriority=16}
  FunctionOutput(None=ScalarParameter() {parameterName=isCriticalHit?}) {outputName=isCriticalHit?, sortPriority=17}
  n92 = ComponentMask(None=TextureCoordinate() {coordinateIndex=3}) {r=true}
  FunctionOutput(None=Divide(A=Subtract(A=@n92) {constB=0.5})) {outputName=Pixel shader Letter Highlight - uv 3 0 to 1, sortPriority=17}
  n93 = Constant3Vector() {constant=<rgba 0 0 1 0>}
  n94 = Add(
   A=RotateAboutAxis(NormalizedRotationAxis=@n87.Result, RotationAngle=Multiply(A=@n79.B) {constB=-1}, PivotPoint=Constant(), Position=@n93)
   B=@n93
  )
  FunctionOutput(None=ComponentMask(None=Transform(None=@n94) {transformSourceType=TRANSFORMSOURCE_World, transformType=TRANSFORM_Tangent}) {g=true, r=true}) {outputName="Custom UV 4 - tangent space up vector (RG)", sortPriority=16}
  FunctionOutput(None=Divide(A=Subtract() {constB=0.5})) {outputName=Vertex Shader - 0-1 highlight pulse, sortPriority=16}
  FunctionOutput(None=MaterialFunctionCall("Normals (Optional)"=FunctionInput() {bUsePreviewValueAsDefault=true, inputName=Normals, previewValue=<0 0 1 1>}) {materialFunction=&Sprite}.Normals) {outputName=Normals, sortPriority=-1}
  n95 = ComponentMask(None=TextureCoordinate() {coordinateIndex=1}) {r=true}
  n96 = Constant() {r=1}
  n97 = Constant()
  FunctionOutput(None=AppendVector(A=If(A=@n95, B=@n96, "A > B"=@n97, "A < B"=@n96), B=@n76.R)) {outputName=CustomUV2, sortPriority=11}
  FunctionOutput(None=@n5) {outputName="optional copy of uv 1 (card selection)", sortPriority=10}
  FunctionOutput(None=@n65.G) {outputName="Random Number  - needs custom uvs", sortPriority=-1}
  n98 = Add(
   A=Multiply(
    A=ScalarParameter() {defaultValue=1, desc=Can turn off the shifting of numbers to camera, parameterName=MoveToCamera, sliderMax=1}
    B=Constant() {r=4.0896499921872486e+18}
   )
   B=FunctionInput() {bUsePreviewValueAsDefault=true, inputName="Additional Camera Offset ", inputType=FunctionInput_Scalar, sortPriority=10}
  )
  n99 = ScalarParameter() {defaultValue=50, parameterName=OffsetFromCamera}
  FunctionOutput(None=Add(
    A=MaterialFunctionCall(Offset Amount=@n98, Clamp Padding=@n99, World Position=@n91) {materialFunction=/Engine/Functions/Engine_MaterialFunctions02/WorldPositionOffset/CameraOffset}.Clamped Camera Offset
    B=@n90
   )) {outputName=WPO}
  FunctionOutput(None=@n92) {outputName=Pixel shader Letter Highlight - uv 3, sortPriority=16}
  FunctionOutput(None=@n68.A) {outputName=CustomUV3 - brightness values for the numbers, sortPriority=12}
  FunctionOutput(None=AppendVector(A=@n48, B=@n5)) {outputName=Custom UV 5 - numbers * .1 and then whole number offsets per card in g, sortPriority=16}
  MaterialFunctionCall() {materialFunction=&DebugScalarValues}
  n100 = Time()
  MaterialFunctionCall!(
   Vector3=TextureSample(UVs=Frac(None=@n100)) {samplerType=SAMPLERTYPE_LinearColor, texture=&ImpactNumbers_XYScale_Rotation_Texture}.RGB
   MaximumNumberOfDigits
   UVs
   DebugTextLocation RG_UpperRight BA_LowerLeft
   Component Spacing
  ) {materialFunction=/Engine/Functions/Engine_MaterialFunctions02/Utility/DebugFloat3Values}
  Multiply(A=@n79.G, B=@n68.G)
  ScalarParameter() {defaultValue=9.323560635889353e+18, parameterName=Camera offset}
  SphereMask(A=Add(A=@n48) {constB=0.05}, B=@n32.G) {attenuationRadius=0.05}
 ]}
